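# Sound cue for an ambient spot of stressed wooden creaks at the Camp Jonesy elevated fort POI, looping with random delays.
# A quality switch selects eleven creak variants plus a weighted empty slot, three variants, or one creak shared by the last two levels.
# Modulators vary volume and pitch, raising pitch markedly on the two randomized branches.
# It overrides attenuation for short range, prevents new instances at the concurrency limit, and uses the Fort_Doors sound class.
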SoundCue /Game/Sounds/Artemis/POI/CampJonesyElevatedFort/Spots/Spot_CampJonesy_Wooden_Stress_Creaks_MediumHigh_Cue
root=Looping_0
nodes [
 {id=Delay_0, children [QualityLevel_0], properties {delayMax=25, delayMin=3}}
 {id=Looping_0, children [Delay_0]}
 {id=Modulator_0, children [WavePlayer_0], properties {pitchMax=1.1, pitchMin=0.8, volumeMax=1, volumeMin=0.3}}
 {id=Modulator_1, children [Random_0], properties {pitchMax=1.6, pitchMin=1.4, volumeMax=1, volumeMin=0.3}}
 {id=Modulator_2, children [Random_1], properties {pitchMax=1.6, pitchMin=1.4, volumeMax=1, volumeMin=0.3}}
 {id=QualityLevel_0, children [Modulator_1,Modulator_2,Modulator_0,Modulator_0]}
 {
  id=Random_0
  children [
   WavePlayer_2
   WavePlayer_3
   WavePlayer_4
   WavePlayer_7
   WavePlayer_6
   WavePlayer_5
   WavePlayer_8
   WavePlayer_9
   WavePlayer_16
   WavePlayer_11
   WavePlayer_12
   null
  ]
  properties {bRandomizeWithoutReplacement=false, weights [0.5,0.5,0.5,0.5,0.5,0.5,0.5,0.5,0.5,0.5,0.5,1]}
 }
 {id=Random_1, children [WavePlayer_14, WavePlayer_17, WavePlayer_13], properties {weights [1,1,1]}}
 {id=WavePlayer_0, properties {soundWaveAssetPtr=/Game/Sounds/Fort_Ambient_NonLoop/Wood/wood_creak_medium_01}}
 {id=WavePlayer_11, properties {soundWaveAssetPtr=/Game/Sounds/Fort_Ambient_NonLoop/Wood/wood_creak_medium_07}}
 {id=WavePlayer_12, properties {soundWaveAssetPtr=/Game/Sounds/Fort_Ambient_NonLoop/Wood/wood_creak_medium_08}}
 {id=WavePlayer_13, properties {soundWaveAssetPtr=/Game/Sounds/Fort_Ambient_NonLoop/Wood/wood_creak_medium_03}}
 {id=WavePlayer_14, properties {soundWaveAssetPtr=&wood_creak_medium_01}}
 {id=WavePlayer_16, properties {soundWaveAssetPtr=/Game/Sounds/Fort_Ambient_NonLoop/Wood/wood_creak_medium_06}}
 {id=WavePlayer_17, properties {soundWaveAssetPtr=&wood_creak_medium_06}}
 {id=WavePlayer_2, properties {soundWaveAssetPtr=/Game/Sounds/Fort_Ambient_NonLoop/Wood/wood_creak_medium_09}}
 {id=WavePlayer_3, properties {soundWaveAssetPtr=/Game/Sounds/Fort_Ambient_NonLoop/Wood/wood_creak_medium_10}}
 {id=WavePlayer_4, properties {soundWaveAssetPtr=/Game/Sounds/Fort_Ambient_NonLoop/Wood/wood_creak_medium_11}}
 {id=WavePlayer_5, properties {soundWaveAssetPtr=&wood_creak_medium_03}}
 {id=WavePlayer_6, properties {soundWaveAssetPtr=/Game/Sounds/Fort_Ambient_NonLoop/Wood/wood_creak_medium_02}}
 {id=WavePlayer_7, properties {soundWaveAssetPtr=&wood_creak_medium_01}}
 {id=WavePlayer_8, properties {soundWaveAssetPtr=/Game/Sounds/Fort_Ambient_NonLoop/Wood/wood_creak_medium_04}}
 {id=WavePlayer_9, properties {soundWaveAssetPtr=/Game/Sounds/Fort_Ambient_NonLoop/Wood/wood_creak_medium_05}}
]
properties {
 bOverrideAttenuation=true
 bOverrideConcurrency=true
 maxDistance=1050
 soundClassObject=&Fort_Doors
 volumeMultiplier=0.8
 attenuationOverrides {falloffDistance=1000, attenuationShapeExtents {x=50}}
 concurrencyOverrides {resolutionRule=PreventNew}
}
